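# Niagara system for the Energystrike vehicle boost cosmetic, a ribbon trail spawned per boost port at skeletal socket positions.
# One merged emitter drives four ribbon renderers, giving each layer its own color, width and dynamic material values through per-particle bindings.
# Particles shoot backward with fading inherited source movement as curves taper width and fade color.
# User parameters switch painted color, per-layer power and hue shift, and narrower frontend-preview widths.
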NiagaraSystem /VehicleCosmetics/Boosts/Boost_Energystrike/FX/NS_Boost_Energystrike
dependencies {
 usedDataInterfaces [&NiagaraDataInterfaceCurve, &NiagaraDataInterfaceSkeletalMesh]
 usedDynamicInputs []
 usedModules [
  &self:SystemSpawnScript.SystemScriptSource.SystemScriptGraph.NiagaraNodeAssignment_0.SetVariables_#1
  &SystemState
 ]
}
emitters {EnergyStrike_Optimized {
  emitterUpdateScript [
   &EmitterState()
   &VC_FilteredSocketCount()
   &VC_RateQualityHandler()
   &VC_SpawnRibbon_v1(Number of Ribbons=$Emitter.PortCount, SpawnRate=80)
  ]
  particleSpawnScript [
   &VC_PaintedVariantsOverride() {enabled=false}
   &VC_InitializeRibbon_v1(
    Custom Offset=true
    Enable Skeletal Position=true
    Lifetime=0.7
    Local Vector Offset=&AddVectorToPosition(Position=<10 0 0>, Vector=$User.DTCarOffset)
    Offset Amount=1
    Ribbon Width:NiagaraFloat=&MakeCustomFloatFromBool(Bool=$User.bIsFrontendPreview, True Float=30, False Float=40)
   )
   &AddVelocity(Velocity=<-1500 0 0>) {module=AddVelocity001}
   &Color(Color=&MakeCustomLinearColorFromBool(Bool=$User.bIsPainted, True Color=$User.PaintedColor, False Color=<rgba 0.102 1 0.102 0.502>))
   /SetVariables(
    Particles.WidthRibbon01:NiagaraFloat=&MakeCustomFloatFromBool(Bool=$User.bIsFrontendPreview, True Float=15, False Float=25)
    Particles.WidthRibbon02:NiagaraFloat=&MakeCustomFloatFromBool(Bool=$User.bIsFrontendPreview, True Float=25, False Float=35)
    Particles.WidthRibbon03:NiagaraFloat=&MakeCustomFloatFromBool(Bool=$User.bIsFrontendPreview, True Float=30, False Float=40)
   ) {isSetParameters=true}
   /VehicleCosmetics/FX/Niagara/Modules/MergeEmitters/VC_ExtraBindingColors(
    ExtraColors=3
    R1 Color:LinearColor=&MakeCustomLinearColorFromBool(Bool=$User.bIsPainted, True Color=$User.PaintedColor, False Color=<rgba 0.6 0.451 0.149 20>)
    R2 Color:LinearColor=&MakeCustomLinearColorFromBool(Bool=$User.bIsPainted, True Color=$User.PaintedColor, False Color=<rgba 0.6 1 0.2 0.502>)
    R3 Color:LinearColor=&MakeCustomLinearColorFromBool(Bool=$User.bIsPainted, True Color=$User.PaintedColor, False Color=<rgba 0.349 1 0.4 0.502>)
   )
  ]
  particleUpdateScript [
   &ParticleState()
   &ScaleVelocity(Velocity Scale=&VectorFromFloat(Value=&FloatFromCurve(FloatCurve=DataInterface() {propertyValues {Curve {keys [<at 0 0>,<at 1 2>]}}}))) {enabled=false}
   &InheritSourceMovement(
    Applied Position Scale=&Multiply_VectorByFloat(
     Vector=<1 1 1>
     Float=&FloatFromCurve(FloatCurve=DataInterface() {propertyValues {Curve {keys [<at 0 1 RCIM_Cubic>, {interpMode=RCIM_Cubic, leaveTangent=-3.4326, time=0.7, value=0}]}}})
    )
    ApplyForce=false
    ApplyPosition=true
    ApplyPositionFalloff=false
   )
   &SolveForcesAndVelocity(Clamp Velocity=false, Limit Acceleration=false)
   &VC_UpdateRibbon_v1(Enable Skeletal Position=true)
   /SetVariables(
    Particles.R1 Color:LinearColor=&VC_DynamicScaleColor(
     Color to Scale:LinearColor=&MakeCustomLinearColorFromBool(Bool=$User.bIsPainted, True Color=$User.PaintedColor, False Color=<rgba 0.6 0.451 0.149 0.5>)
     RGB Scale:Vector3f=&VectorFromFloat(Value=&FloatFromCurve(
       FloatCurve=DataInterface() {propertyValues {Curve {keys [<at 0 0.35>,<at 1 0>]}}}
       Scale Curve=&MakeCustomFloatFromBool(Bool=$User.bIsPainted, True Float=$User.Power_EmissivePaintedR2, False Float=1)
      ))
     Alpha Scale:NiagaraFloat=&FloatFromCurve(
      FloatCurve=DataInterface() {propertyValues {Curve {keys [
          <at 0 1 RCIM_Cubic>
          {arriveTangent=-0.2383, interpMode=RCIM_Cubic, leaveTangent=-0.2383, tangentMode=RCTM_User, time=0.35, value=1}
          <at 1 0 RCIM_Cubic>
         ]}}}
      Scale Curve=&MakeCustomFloatFromBool(Bool=$User.bIsPainted, True Float=$User.Power_AlphaPaintedR2, False Float=1)
     )
    )
    Particles.R2 Color:LinearColor=&VC_DynamicScaleColor(
     Color to Scale:LinearColor=&MakeCustomLinearColorFromBool(Bool=$User.bIsPainted, True Color=$User.PaintedColor, False Color=<rgba 0.6 1 0.2 0.502>)
     RGB Scale:Vector3f=&VectorFromFloat(Value=&FloatFromCurve(
       FloatCurve=DataInterface() {propertyValues {Curve {keys [<at 0 2 RCIM_Cubic>,<at 1 0 RCIM_Cubic>]}}}
       Scale Curve=&MakeCustomFloatFromBool(Bool=$User.bIsPainted, True Float=$User.Power_EmissivePaintedR3, False Float=2)
      ))
     Alpha Scale:NiagaraFloat=&FloatFromCurve(
      FloatCurve=DataInterface() {propertyValues {Curve {keys [
          <at 0 5 RCIM_Cubic>
          {arriveTangent=-0.2383, interpMode=RCIM_Cubic, leaveTangent=-0.2383, tangentMode=RCTM_User, time=0.8, value=1}
          <at 1 0 RCIM_Cubic>
         ]}}}
      Scale Curve=&MakeCustomFloatFromBool(Bool=$User.bIsPainted, True Float=$User.Power_AlphaPaintedR3, False Float=1)
     )
    )
    Particles.R3 Color:LinearColor=&VC_DynamicScaleColor(
     Color to Scale:LinearColor=&MakeCustomLinearColorFromBool(Bool=$User.bIsPainted, True Color=$User.PaintedColor, False Color=<rgba 0.349 1 0.4 0.502>)
     RGB Scale:Vector3f=&VectorFromFloat(Value=&FloatFromCurve(
       FloatCurve=DataInterface() {propertyValues {Curve {keys [<at 0 8 RCIM_Cubic>,<at 1 1 RCIM_Cubic>]}}}
       Scale Curve=&MakeCustomFloatFromBool(Bool=$User.bIsPainted, True Float=$User.Power_EmissivePaintedR4, False Float=1)
      ))
     Alpha Scale:NiagaraFloat=&FloatFromCurve(
      FloatCurve=DataInterface() {propertyValues {Curve {keys [
          <at 0 8 RCIM_Cubic>
          {arriveTangent=-0.2383, interpMode=RCIM_Cubic, leaveTangent=-0.2383, tangentMode=RCTM_User, time=0.5, value=8}
          <at 1 0 RCIM_Cubic>
         ]}}}
      Scale Curve=&MakeCustomFloatFromBool(Bool=$User.bIsPainted, True Float=$User.Power_AlphaPaintedR4, False Float=1.5)
     )
    )
   ) {isSetParameters=true}
   &ScaleColor(
    Scale Alpha=&FloatFromCurve(
     FloatCurve=DataInterface() {propertyValues {Curve {keys [
         <at 0 2 RCIM_Cubic>
         {arriveTangent=-0.2383, interpMode=RCIM_Cubic, leaveTangent=-0.2383, tangentMode=RCTM_User, time=0.8, value=1}
         <at 1 0 RCIM_Cubic>
        ]}}}
     Scale Curve=&MakeCustomFloatFromBool(Bool=$User.bIsPainted, True Float=$User.Power_AlphaPaintedR1, False Float=1)
    )
    Scale RGB=&VectorFromFloat(Value=&FloatFromCurve(
      FloatCurve=DataInterface() {propertyValues {Curve {keys [<at 0 1>,<at 1 0>]}}}
      Scale Curve=&MakeCustomFloatFromBool(Bool=$User.bIsPainted, True Float=$User.Power_EmissivePaintedR1, False Float=1)
     ))
    ScaleA=true
    ScaleRGB=true
   )
   &DynamicMaterialParameters(
    Index 0 Param 1=5
    Index 0 Param 2=0.5
    Index 0 Param 4=1
    Index 1 Param 2=&MakeCustomFloatFromBool(Bool=$User.bIsPainted, True Float=$User.HueShift_M1, False Float=0.05)
    Module.Index0Param0Write=true
    Module.Index0Param1Write=true
    Module.Index0Param2Write=true
    Module.Index0Param3Write=true
    Module.Index1Param0Write=true
    Module.Index1Param1Write=true
    Module.Index1Param2Write=false
    Module.Index1Param3Write=false
    Write Parameter Index 1=true
   )
   &ScaleRibbonWidth(Ribbon Width Scale=&FloatFromCurve(FloatCurve=DataInterface() {propertyValues {Curve {keys [{interpMode=RCIM_Cubic, tangentMode=RCTM_Break, time=0, value=3}, {interpMode=RCIM_Cubic, tangentMode=RCTM_Break, time=1, value=0.5}]}}}))
   /SetVariables(
    Particles.Index_Dynamic1_01:Vector4f=&MakeVector4(X=1, Y=0.1, Z=1, W=10)
    Particles.Index_Dynamic2_01:Vector4f=&MakeVector4(
     X=0
     Y=&MakeCustomFloatFromBool(Bool=$User.bIsPainted, True Float=$User.HueShift_M4, False Float=0.2)
     Z=0
     W=0
    )
   ) {isSetParameters=true}
   /SetVariables(
    Particles.Index_Dynamic1_02:Vector4f=&MakeVector4(X=1, Y=1, Z=1, W=3)
    Particles.Index_Dynamic2_02:Vector4f=&MakeVector4(
     X=0
     Y=&MakeCustomFloatFromBool(Bool=$User.bIsPainted, True Float=$User.HueShift_M2, False Float=0)
     Z=0
     W=0
    )
   ) {isSetParameters=true}
   /SetVariables(
    Particles.Index_Dynamic1_03:Vector4f=&MakeVector4(X=0.35, Y=1, Z=0.05, W=5)
    Particles.Index_Dynamic2_03:Vector4f=&MakeVector4(
     X=0
     Y=&MakeCustomFloatFromBool(Bool=$User.bIsPainted, True Float=$User.HueShift_M3, False Float=0.06)
     Z=0
     W=0
    )
   ) {isSetParameters=true}
   /SetVariables(
    Particles.ScaleWidthR01:NiagaraFloat=&Multiply_Float(
     A=$Particles.WidthRibbon01
     B=&FloatFromCurve(FloatCurve=DataInterface() {propertyValues {Curve {keys [{interpMode=RCIM_Cubic, tangentMode=RCTM_Break, time=0, value=3}, {interpMode=RCIM_Cubic, tangentMode=RCTM_Break, time=1, value=0.5}]}}})
    )
    Particles.ScaleWidthR02:NiagaraFloat=&Multiply_Float(
     A=$Particles.WidthRibbon02
     B=&FloatFromCurve(FloatCurve=DataInterface() {propertyValues {Curve {keys [{interpMode=RCIM_Cubic, tangentMode=RCTM_Break, time=0, value=3}, {interpMode=RCIM_Cubic, tangentMode=RCTM_Break, time=1, value=0.5}]}}})
    )
    Particles.ScaleWidthR03:NiagaraFloat=&Multiply_Float(
     A=$Particles.WidthRibbon03
     B=&FloatFromCurve(FloatCurve=DataInterface() {propertyValues {Curve {keys [{interpMode=RCIM_Cubic, tangentMode=RCTM_Break, time=0, value=3}, {interpMode=RCIM_Cubic, tangentMode=RCTM_Break, time=1, value=0.5}]}}})
    )
   ) {isSetParameters=true}
  ]
  rendererClasses [&NiagaraRibbonRendererProperties]
  renderers [
   {rendererClass=&NiagaraRibbonRendererProperties}
   {rendererClass=&NiagaraRibbonRendererProperties, rendererIndex=1}
   {rendererClass=&NiagaraRibbonRendererProperties, rendererIndex=2}
   {rendererClass=&NiagaraRibbonRendererProperties, rendererIndex=3}
  ]
 }}
summary {
 systemName=NS_Boost_Energystrike
 emitters [{emitterName=EnergyStrike_Optimized, rendererClasses [&NiagaraRibbonRendererProperties]}]
 userVariables [
  User.bIsFrontendPreview {type=NiagaraBool, defaultValue=true}
  User.bIsPainted {type=NiagaraBool}
  User.ColorVariantID {type=NiagaraInt32}
  User.DTCarOffset {type=Vector3f, description="", defaultValue {struct=Vector3f, value=<0 0 0>}}
  User.HueShift_M1 {type=NiagaraFloat, defaultValue=0.05, description=""}
  User.HueShift_M2 {type=NiagaraFloat, defaultValue=0, description=""}
  User.HueShift_M3 {type=NiagaraFloat, defaultValue=0.05, description=""}
  User.HueShift_M4 {type=NiagaraFloat, defaultValue=0.2, description=""}
  User.PaintedColor {type=LinearColor, description="", defaultValue {struct=LinearColor, value=<rgba 1 1 1 1>}}
  User.PaintedScalar {type=NiagaraFloat, defaultValue=1, description=""}
  User.PortCount {type=NiagaraInt32, defaultValue=1}
  User.Power_AlphaPaintedR1 {type=NiagaraFloat, defaultValue=1, description=""}
  User.Power_AlphaPaintedR2 {type=NiagaraFloat, defaultValue=1, description=""}
  User.Power_AlphaPaintedR3 {type=NiagaraFloat, defaultValue=1, description=""}
  User.Power_AlphaPaintedR4 {type=NiagaraFloat, defaultValue=1, description=""}
  User.Power_EmissivePaintedR1 {type=NiagaraFloat, defaultValue=1, description=""}
  User.Power_EmissivePaintedR2 {type=NiagaraFloat, defaultValue=1, description=""}
  User.Power_EmissivePaintedR3 {type=NiagaraFloat, defaultValue=1, description=""}
  User.Power_EmissivePaintedR4 {type=NiagaraFloat, defaultValue=1, description=""}
 ]
}
