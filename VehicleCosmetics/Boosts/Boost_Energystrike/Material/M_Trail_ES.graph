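# Unlit translucent Niagara ribbon material for the Energystrike boost trail.
# Panned distortion noise warps an edge mask and two caustics layers, which dynamic parameters shape into emissive and opacity, faded along U.
# Hue shift and desaturation can vary with particle red and green via static switches.
# Optional switches add distortion from the masked caustics and a sine/cosine world position offset wobble.
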Material /VehicleCosmetics/Boosts/Boost_Energystrike/Material/M_Trail_ES
graph {
 properties {
  bUsedWithNiagaraRibbons=true
  blendMode=BLEND_Translucent
  pixelDepthOffsetMode=PDOM_Legacy
  refractionMethod=RM_2DOffset
  shadingModel=MSM_Unlit
  usedShadingModels=Unlit
 }
 nodes [
  n1 = DynamicParameter() {parameterIndex=1, paramNames [Saturation, HueShift, "", ""]}
  n2 = Multiply(A=Constant() {r=0.1}, B=@n1.HueShift)
  n3 = ParticleColor()
  n4 = LinearInterpolate(A=@n1.HueShift, B=@n2, Alpha=@n3.R)
  n5 = LinearInterpolate(A=@n2, B=@n1.HueShift, Alpha=Add(A=@n3.G, B=@n3.G))
  n6 = StaticSwitchParameter(True=Saturate(None=Add(A=@n4, B=@n5)), False=@n1.HueShift) {defaultValue=true, parameterName=UseHueOffset}
  n7 = Constant() {r=0.5}
  n8 = TextureCoordinate() {vTiling=0.125}
  n9 = Constant() {r=0.05}
  n10 = TextureSample(UVs=Panner(Coordinate=TextureCoordinate() {uTiling=0.25, vTiling=0.25}) {speedX=-0.125, speedY=0.01}) {texture=/DealershipSharedLibrary/FX/Textures/Noise/T_VCVFX_Flow02_N}
  n11 = Multiply(A=Constant() {r=0.05}, B=AppendVector(A=@n10.R, B=@n10.R))
  n12 = Panner(Coordinate=Add(A=Panner(Coordinate=@n8, Time=@n9) {speedY=0.5}, B=@n11)) {speedX=-1}
  n13 = TextureCoordinate() {uTiling=1.25, vTiling=-0.125}
  n14 = Panner(Coordinate=Add(A=Panner(Coordinate=@n13, Time=@n9) {speedY=-0.5}, B=@n11)) {speedX=-1}
  n15 = ComponentMask(None=Add(
    A=TextureSample(UVs=@n12) {texture=/DealershipSharedLibrary/FX/Textures/Noise/T_VCVFX_DistortNoise02_RGB}.RGB
    B=TextureSample(UVs=@n14) {texture=&T_VCVFX_DistortNoise02_RGB}.RGB
   )) {g=true, r=true}
  n16 = ComponentMask(None=Add(A=TextureCoordinate(), B=Multiply(A=@n7, B=@n15))) {g=true}
  n17 = OneMinus(None=@n16)
  n18 = Multiply(A=Saturate(None=ConstantBiasScale(None=@n16) {bias=-0.15, scale=1}), B=Saturate(None=ConstantBiasScale(None=@n17) {bias=-0.01, scale=1}))
  n19 = DynamicParameter() {paramNames [Hardness,SolidMask,AddMultiply,Brightness]}
  n20 = Multiply(A=Saturate(None=Multiply(A=@n18, B=@n19.Hardness)), B=@n19.Brightness)
  n21 = Constant() {r=6}
  n22 = LinearInterpolate(A=@n20, B=Saturate(None=Multiply(A=@n21, B=@n18)), Alpha=@n19.SolidMask)
  n23 = Clamp(None=@n22) {maxDefault=1.5}
  n24 = ParticleColor()
  n25 = ScalarParameter() {parameterName=Saturation}
  n26 = Multiply(A=Constant() {r=0.5}, B=@n25)
  n27 = ParticleColor()
  n28 = Add(A=LinearInterpolate(A=@n25, B=@n26, Alpha=@n27.R), B=LinearInterpolate(A=@n26, B=@n25, Alpha=Add(A=@n27.G, B=@n27.G)))
  n29 = Desaturation(
   None=MaterialFunctionCall(Hue Shift Percentage=Multiply(A=@n6, B=@n23), Texture=@n24.RGB) {materialFunction=&HueShift}.Result
   Fraction=StaticSwitchParameter(True=Saturate(None=@n28), False=@n25) {defaultValue=true, parameterName=UseSaturationOffset}
  )
  MaterialFunctionCall(Emissive=Multiply(A=Constant() {r=1.8}, B=Multiply(A=@n29, B=@n22))) {materialFunction=&MF_FX_EmissiveAdjustments}
  n30 = Constant() {r=0.5}
  n31 = Multiply(A=Constant() {r=0.15}, B=@n15)
  n32 = Add(A=Panner(Coordinate=TextureCoordinate() {uTiling=1.25, vTiling=0.5}, Time=@n30) {speedY=0.5}, B=@n31)
  n33 = Multiply(A=Time(), B=ScalarParameter() {defaultValue=2, parameterName=PanSpeed})
  n34 = StaticComponentMaskParameter(None=TextureSample(UVs=Panner(Coordinate=@n32, Time=@n33) {speedX=-1, speedY=0.25}) {texture=/DealershipSharedLibrary/FX/Textures/Noise/T_VCVFX_CausticsMask_RGB}.RGB) {defaultG=true, parameterName=Mask_RGB01}
  n35 = Add(A=@n31, B=Panner(Coordinate=TextureCoordinate() {vTiling=-0.5}, Time=@n30) {speedY=-0.5})
  n36 = StaticComponentMaskParameter(None=TextureSample(UVs=Panner(Coordinate=@n35, Time=@n33) {speedX=-0.5, speedY=0.2}) {texture=&T_VCVFX_CausticsMask_RGB}.RGB) {defaultB=true, parameterName=Mask_RGB02}
  n37 = LinearInterpolate(A=Multiply(A=Multiply(A=@n34, B=@n36), B=@n19.Brightness), B=Multiply(A=@n19.Brightness, B=Add(A=@n34, B=@n36)), Alpha=@n19.AddMultiply)
  n38 = Multiply(A=Saturate(None=Multiply(A=@n22, B=@n37)), B=@n19.Hardness)
  n39 = Multiply(A=@n24.A, B=LinearInterpolate(A=@n20, B=Saturate(None=@n38), Alpha=@n19.SolidMask))
  n40 = ComponentMask(None=TextureCoordinate()) {r=true}
  n41 = Saturate(None=ConstantBiasScale(None=@n40) {bias=-0.01, scale=8})
  n42 = TextureCoordinate()
  Multiply(A=Saturate(None=Multiply(A=@n39, B=@n41)), B=Power(Base=ComponentMask(None=@n42) {r=true}) {constExponent=0.5})
  n43 = Multiply(A=@n18, B=@n37)
  n44 = Constant() {r=5}
  n45 = Saturate(None=OneMinus(None=@n40))
  StaticSwitchParameter(True=Multiply(A=Multiply(A=@n43, B=@n44), B=Multiply(A=@n41, B=@n45)), False=Constant()) {parameterName=Distortion}
  n46 = Multiply(A=ScalarParameter() {parameterName=speed}, B=OneMinus(None=Time()))
  n47 = ComponentMask(None=Multiply(A=TextureCoordinate() {coordinateIndex=1, uTiling=3}, B=ScalarParameter() {parameterName=Tile})) {r=true}
  n48 = Add(A=@n46, B=@n47)
  n49 = Multiply(A=@n46, B=Constant() {r=0.5})
  n50 = Add(A=Add(A=Cosine(None=@n48), B=Cosine(None=@n48) {period=2}), B=Cosine(None=Add(A=@n47, B=@n49)) {period=0.75})
  n51 = Add(A=@n47, B=@n46)
  n52 = Sine(None=@n51)
  n53 = Sine(None=@n51) {period=2}
  n54 = AppendVector(A=@n50, B=Add(A=Add(A=@n52, B=@n53), B=Sine(None=@n51) {period=0.75}))
  n55 = Multiply(
   A=Multiply(A=AppendVector(A=@n54, B=@n50), B=ScalarParameter() {parameterName=amount})
   B=ScalarParameter() {defaultValue=1, parameterName=Intensity}
  )
  StaticSwitchParameter(True=Multiply(A=Saturate(None=@n41), B=Transform(None=@n55)), False=Constant()) {parameterName=WPO}
 ]
}
